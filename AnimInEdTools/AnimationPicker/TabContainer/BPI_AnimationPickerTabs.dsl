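; Blueprint interface for tab widgets in the Animation Picker tab container of the AnimInEd editor animation tools.
; It declares one function, SetControlRigBindingProxy, which takes a Control Rig Sequencer Binding Proxy struct.
; Implementing tabs can receive a Control Rig's Sequencer binding through this call.
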(fn SetControlRigBindingProxy
  (params (Control%20Rig%20Binding%20Proxy :type Control%20Rig%20Sequencer%20Binding%20Proxy%20Structure)))
