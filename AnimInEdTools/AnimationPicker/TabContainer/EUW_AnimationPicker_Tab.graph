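# Editor Utility Widget tab container for the Animation Picker, hosting one picker per tab with add and remove controls.
# The SG_AnimationPicker save game stores each tab's index, Control Rig binding and bookmarks, rebuilding tabs on Construct and restoring valid bindings on selection.
# Bookmarks save the picker's MoverCanvasPanel render transform in nine context-menu slots, recalled by menu or by key press matched on key display name.
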EditorUtilityWidgetBlueprint /AnimInEdTools/AnimationPicker/TabContainer/EUW_AnimationPicker_Tab
graphs [
 EventGraph {nodes [
   "On Clicked (AddTabButton)"() {is=title K2Node_ComponentBoundEvent}
   > n1 = Create Widget!(
    execute:Exec
    WorldContextObject:Object Reference
    Class:User Widget Class Reference=&/EUW_AnimationPickerTabManager.EUW_AnimationPickerTabManager_C
    OwningPlayer:Player Controller Object Reference
   ) -> (then:Exec, ReturnValue:EUW Animation Picker Tab Manager Object Reference)
   > n2 = AddChild(self=Get DialogSlot() -> (DialogSlot:Named Slot Object Reference), Content=@n1)
   n3 = Add Tab() {is=title K2Node_CustomEvent}
   Bind Event to Add Tab Clicked Dispatcher(execute=@n2.then, self:EUW Animation Picker Tab Manager Object Reference=@n1, Delegate=@n3) {is=title K2Node_AddDelegate}
   "On Clicked (RemoveTabButton)"() {is=title K2Node_ComponentBoundEvent}
   > n4 = getSelectedTab!(execute, self:EUW Base Tab Container Object Reference) -> (then, index:int=0, name:String)
   > removeTabByName!(execute, self:EUW Base Tab Container Object Reference, nameToRemove:String=@n4.name) -> (then)
   > n5 = LoadGameFromSlot!(execute, SlotName:String=Animation Picker, UserIndex:int=0) -> (then, ReturnValue:Save Game Object Reference)
   > n6 = Cast To SG_AnimationPicker!(execute, Object=@n5) -> (then, CastFailed, AsSG Animation Picker:SG Animation Picker Object Reference) {is=title K2Node_DynamicCast}
   > Remove!(
    execute
    TargetMap:Map of Strings to STRUCT Picker Tab Data Structures=Get Tabs ControlRig Binding!(self:SG Animation Picker Object Reference=@n6) -> (Tabs ControlRig Binding:Map of Strings to STRUCT Picker Tab Data Structures)
    Key:"String (by ref)"=@n4.name
   ) -> (then, ReturnValue:bool)
   > SaveGameToSlot!(execute, SaveGameObject:Save Game Object Reference=@n6, SlotName:String=Animation Picker, UserIndex:int=0) -> (then, ReturnValue:bool)
   "On Clicked (SetBookmarkButton)"() {is=title K2Node_ComponentBoundEvent}
   > Clear Context Menu(self:EUW Context Menu Object Reference=Get SetBookmark_ContextMenu() -> (SetBookmark_ContextMenu:EUW Context Menu Object Reference))
   > n7 = Sequence()
   > n8 = For Loop(LastIndex=8)
   n9 = Set Bookmark Event!() -> (OutputDelegate, then, Context Item:Editor Utility Widget Object Reference, Slot Widget:Editor Utility Widget Object Reference) {is=title K2Node_CustomEvent}
   n10 = Add Item!(
    execute=@n8.LoopBody
    self:EUW Context Menu Object Reference=Get SetBookmark_ContextMenu()
    Event:"Delegate (by ref)"=@n9
    Slot Widget:User Widget Class Reference=/AnimInEdTools/ContextMenu/EUW_ContextMenu_ItemB.EUW_ContextMenu_ItemB_C
    Category:String
   ) -> (then, Context Menu Item:EUW Context Menu Item Object Reference, Widget:Editor Utility Widget Object Reference)
   > n11 = Cast To EUW_ContextMenu_ItemB!(execute, Object=@n10.Widget) -> (then, CastFailed, AsEUW Context Menu Item B:EUW Context Menu Item B Object Reference) {is=title K2Node_DynamicCast}
   n12 = "To String (Integer)"(InInt=int + int(A=@n8))
   > Set Text(
    Text="To Text (String)"(InString=Append(A=Set Bookmark, B=@n12))
    self=Get ButtonText!(self:EUW Context Menu Item B Object Reference=@n11) -> (ButtonText:Text Object Reference)
   )
   n13 = Get SetBookmarkButton() -> (SetBookmarkButton:Editor Utility Button Object Reference)
   n14 = BreakVector2D(InVec=GetAbsolutePosition!(Geometry:"Geometry Structure (by ref)"=GetCachedGeometry!(self:Widget Object Reference=@n13) -> (ReturnValue:"Geometry Structure (by ref)")) -> (ReturnValue:Vector 2D Structure))
   Show!(
    execute=@n7.then_1
    self:EUW Context Menu Object Reference=Get SetBookmark_ContextMenu()
    Show:bool=true
    Mouse Position:Vector 2D Structure=MakeVector2D(X=@n14, Y=float - float(A=@n14.Y, B=30))
    Name:String
   ) -> (then)
   "On Clicked (GoToBookmarkButton)"() {is=title K2Node_ComponentBoundEvent}
   > Clear Context Menu(self=Get SetBookmark_ContextMenu())
   > n15 = Sequence()
   > n16 = For Loop(LastIndex=8)
   n17 = Go To Bookmark Event!() -> (OutputDelegate, then, Context Item:Editor Utility Widget Object Reference, Slot Widget:Editor Utility Widget Object Reference) {is=title K2Node_CustomEvent}
   n18 = Add Item(
    execute=@n16.LoopBody
    self=Get SetBookmark_ContextMenu()
    Event=@n17
    Slot Widget=&EUW_ContextMenu_ItemB_C
   )
   > n19 = Cast To EUW_ContextMenu_ItemB(Object=@n18.Widget)
   n20 = "To String (Integer)"(InInt=int + int(A=@n16))
   > Set Text(Text="To Text (String)"(InString=Append(A=Go To Bookmark, B=@n20)), self=Get ButtonText(self=@n19))
   n21 = Get SetBookmarkButton()
   n22 = BreakVector2D(InVec=GetAbsolutePosition(Geometry=GetCachedGeometry(self=@n21)))
   Show(execute=@n15.then_1, self=Get SetBookmark_ContextMenu(), Show=true, Mouse Position=MakeVector2D(X=@n22, Y=float - float(A=@n22.Y, B=30)))
   n23 = TabAdded!() -> (
    OutputDelegate
    then
    TabContainer:EUW Base Tab Container Object Reference
    NewTabName_tabName_2_#1:String
    NewTabName_tabIndex_5_#2:int
    NewTabName_tabWidgetClass_11_#3:User Widget Class Reference
    NewTabName_tabContentWidget_14_#4:User Widget Object Reference
    NewTabName_tabButtonWidget_17_#5:User Widget Object Reference
    NewTabName_tabButtonStyle_20_#6:TCW Button Visuals Struct Structure
   ) {is=title K2Node_CustomEvent}
   > setTabSelectedByIndex!(execute, self:EUW Base Tab Container Object Reference, indexToSelect:int=@n23.NewTabName_tabIndex_5_#2) -> (then)
   > Delay!(execute, WorldContextObject:Object Reference, Duration=1) -> (then)
   > n24 = Cast To Animation_Picker!(execute, Object=@n23.NewTabName_tabContentWidget_14_#4) -> (then, CastFailed, AsAnimation Picker:Animation Picker Object Reference) {is=title K2Node_DynamicCast}
   n25 = CharacterContextMenuChanged!() -> (OutputDelegate, then, Current ControlRig Binding:Control Rig Sequencer Binding Proxy Structure) {is=title K2Node_CustomEvent}
   Bind Event to Character Context Menu Changed Dispatcher(execute=@n24.then, self:Animation Picker Object Reference=@n24, Delegate=@n25) {is=title K2Node_AddDelegate}
   > Update Tab Settings GUID!(
    execute
    self
    TabName:String=@n23.NewTabName_tabName_2_#1
    TabIndex:int=@n23.NewTabName_tabIndex_5_#2
    ControlRig:Control Rig Sequencer Binding Proxy Structure=Get Current Control Rig Binding!(self:Animation Picker Object Reference=@n24) -> (Current Control Rig Binding:Control Rig Sequencer Binding Proxy Structure)
   ) -> (then)
   > SetKeyboardFocus(self:Widget Object Reference=@n24)
   > Branch(Condition="integer > integer"(A=@n23.NewTabName_tabIndex_5_#2))
   > Set CheckedState!(
    execute
    CheckedState:ECheckBoxState Enum=Unchecked
    self:Check Box Object Reference=Get Auto_CheckBox!(self:Animation Picker Object Reference=@n24) -> (Auto_CheckBox:Editor Utility Check Box Object Reference)
   ) -> (then, Output_Get:ECheckBoxState Enum=Unchecked)
   n26 = OnTabSelectedEvent!() -> (OutputDelegate, then, tabIndex:int, tabWidget:Widget Object Reference) {is=title K2Node_CustomEvent}
   > n27 = getSelectedTab() -> (index=0)
   > n28 = Cast To Animation_Picker(Object=@n26.tabWidget)
   > Set Selected Tab_Event!(execute, self:Animation Picker Object Reference=@n28, Tab Name:String=@n27.name, Tab Index:int=@n27) -> (then) {is=title K2Node_CallFunction}
   > DelayUntilNextTick!(execute, WorldContextObject:Object Reference) -> (then)
   > n29 = LoadGameFromSlot(SlotName=Animation Picker, UserIndex=0)
   > n30 = Cast To SG_AnimationPicker(Object=@n29)
   > n31 = Sequence()
   n32 = Break STRUCT Picker Tab Data!(STRUCT_PickerTabData:"STRUCT Picker Tab Data Structure (by ref)"=Find(TargetMap:Map of Strings to STRUCT Picker Tab Data Structures=Get Tabs ControlRig Binding(self=@n30), Key=@n27.name) -> (Value:STRUCT Picker Tab Data Structure)) -> (
    ControlRigBinding_13_#7:Control Rig Sequencer Binding Proxy Structure
    Index_16_#8:int
    Bookmarks_17_#9:Map of Strings to Widget Transform Structures
   ) {is=title K2Node_BreakStruct}
   > Branch(Condition=IsValid(Object=Break Control Rig Sequencer Binding Proxy!(ControlRigSequencerBindingProxy:"Control Rig Sequencer Binding Proxy Structure (by ref)"=@n32) -> (
      Proxy:Movie Scene Binding Proxy Structure
      ControlRig:Control Rig Object Reference
      Track:Movie Scene Control Rig Parameter Track Object Reference
     ) {is=title K2Node_BreakStruct}.Track))
   > n33 = Set Current Control Rig Binding!(execute, Current Control Rig Binding:Control Rig Sequencer Binding Proxy Structure=@n32, self:Animation Picker Object Reference=@n28) -> (then, Output_Get:Control Rig Sequencer Binding Proxy Structure)
   > Branch(Condition=Does Object Implement Interface!(
     TestObject:Object Reference=@n28
     Interface:Interface Interface=&/BPI_AnimationPickerTabs.BPI_AnimationPickerTabs_C
    ) -> (ReturnValue:bool))
   > Set ControlRig Binding Proxy!(execute, self:BPI Animation Picker Tabs Interface=@n28, Control Rig Binding Proxy:Control Rig Sequencer Binding Proxy Structure=@n33) -> (then) {is=title K2Node_CallFunction}
   n34 = SetKeyboardFocus(execute=@n31.then_1, self=@n28)
   n35 = Combo Changed!() -> (OutputDelegate, then, Current ControlRig Binding:Control Rig Sequencer Binding Proxy Structure) {is=title K2Node_CustomEvent}
   Bind Event to Character Context Menu Changed Dispatcher(execute=@n34.then, self=@n28, Delegate=@n35)
   > Update Tab Settings GUID(TabName=@n27.name, TabIndex=@n27, ControlRig=@n32)
   Event Construct()
   n36 = Make Margin(Left=0, Top=8, Right=0, Bottom=0)
   n37 = Get SetBookmark_ContextMenu()
   > Set ScrollbarPadding!(
    execute
    ScrollbarPadding:Margin Structure=@n36
    self:Scroll Box Object Reference=Get ScrollBox!(self:EUW Context Menu Object Reference=@n37) -> (ScrollBox:Scroll Box Object Reference)
   ) -> (then, Output_Get:Margin Structure)
   > Set Padding!(
    execute
    Padding:Margin Structure=@n36
    self:Border Object Reference=Get Border!(self:EUW Context Menu Object Reference=@n37) -> (Border:Border Object Reference)
   ) -> (then, Output_Get:Margin Structure)
   > SetVisibility(self=Get SetBookmark_ContextMenu(), InVisibility=Collapsed)
   > Set tabButtonParent(tabButtonParent:Widget Object Reference=Get listTabs() -> (listTabs:TA Generic 01 List Object Reference)) -> (Output_Get:Widget Object Reference)
   > Set tabWidgetParent(tabWidgetParent:Widget Object Reference=Get widgetSwitcherTabWidgets() -> (widgetSwitcherTabWidgets:Common Widget Switcher Legacy Object Reference)) -> (Output_Get:Widget Object Reference)
   > Parent: Construct()
   > n38 = Sequence!(execute) -> (then_0, then_1, then_2:Exec)
   Bind Event to On Tab Selected(execute=@n38.then_1, self:EUW Base Tab Container Object Reference, Delegate=@n26) {is=title K2Node_AddDelegate}
   n39 = LoadGameFromSlot(execute=@n38.then_2, SlotName=Animation Picker, UserIndex=0)
   > n40 = Cast To SG_AnimationPicker(Object=@n39)
   n41 = Get Tabs ControlRig Binding(self=@n40)
   > n42 = Branch(Condition=Is Empty!(TargetMap:Map of Strings to STRUCT Picker Tab Data Structures=@n41) -> (ReturnValue:bool) {is=title K2Node_CallFunction})
   addTab!(
    execute=[@n42.then, @n40.CastFailed]
    self:EUW Base Tab Container Object Reference
    tabName:String=Default
    tabWidgetToLoad:User Widget Class Reference=/AnimInEdTools/AnimationPicker/Animation_Picker.Animation_Picker_C
    insertIndex:int=0
    ValidateVisuals:bool=true
    OptionalTabStyle:TCW Button Visuals Struct Structure
   ) -> (
    then
    newTab_tabName_2_#1:String
    newTab_tabIndex_5_#2:int
    newTab_tabWidgetClass_11_#3:User Widget Class Reference
    newTab_tabContentWidget_14_#4:User Widget Object Reference
    newTab_tabButtonWidget_17_#5:User Widget Object Reference
    newTab_tabButtonStyle_20_#6:TCW Button Visuals Struct Structure
   )
   n43 = "For Each Loop (Map)"!(execute:Exec=@n42.else, MapPin:Map of Strings to STRUCT Picker Tab Data Structures=@n41, BreakPin:Exec) -> (
    then:Exec
    KeyPin:String
    ValuePin_ControlRigBinding_13_#7:Control Rig Sequencer Binding Proxy Structure
    ValuePin_Index_16_#8:int
    ValuePin_Bookmarks_17_#9:Map of Strings to Widget Transform Structures
    CompletedPin:Exec
   )
   > addTab(
    tabName=@n43
    tabWidgetToLoad=&Animation_Picker_C
    insertIndex=@n43.ValuePin_Index_16_#8
    ValidateVisuals=true
   )
   Bind Event to On Tab Added(execute=@n38.then_0, self:EUW Base Tab Container Object Reference, Delegate=@n23) {is=title K2Node_AddDelegate}
   n44 = Event Tick()
   > Parent: Tick!(execute, MyGeometry:Geometry Structure=@n44.MyGeometry, InDeltaTime:float=@n44.InDeltaTime) -> (then) {is=title K2Node_CallParentFunction}
   n45 = LoadGameFromSlot(execute=@n60.then, SlotName=Animation Picker, UserIndex=0)
   > n46 = Cast To SG_AnimationPicker(Object=@n45)
   SaveGameToSlot(execute=@n66.then, SaveGameObject=@n46, SlotName=Animation Picker, UserIndex=0)
   n47 = Cast To EUW_ContextMenu_ItemB(execute=@n17.then, Object=@n17.Slot Widget)
   n48 = Get Text(self:Text Object Reference=Get ButtonText(self=@n47))
   > n49 = getSelectedTab() -> (index=0)
   n50 = LoadGameFromSlot(execute=@n54.then, SlotName=Animation Picker, UserIndex=0)
   > n51 = Cast To SG_AnimationPicker(Object=@n50)
   n52 = Find(
    TargetMap:Map of Strings to Widget Transform Structures=Find!(TargetMap:Map of Strings to STRUCT Picker Tab Data Structures=Get Tabs ControlRig Binding(self=@n51), Key=@n49.name) -> (
     Value_ControlRigBinding_13_#7:Control Rig Sequencer Binding Proxy Structure
     Value_Index_16_#8:int
     Value_Bookmarks_17_#9:Map of Strings to Widget Transform Structures
     ReturnValue
    ).Value_Bookmarks_17_#9
    Key=Replace(SourceString="To String (Text)"(InText=@n48), From="Go To ")
   ) -> (Value:Widget Transform Structure)
   n53 = getTabWidgetByIndex!(execute=@n49.then, self:EUW Base Tab Container Object Reference, index:int=@n49) -> (then, widget:Widget Object Reference)
   > n54 = Cast To Animation_Picker(Object=@n53)
   SetRenderTransform!(
    execute=@n67.then
    self:Widget Object Reference=Get MoverCanvasPanel!(self:Animation Picker Object Reference=@n54) -> (MoverCanvasPanel:Canvas Panel Object Reference)
    InTransform:Widget Transform Structure=@n52
   ) -> (then)
   n55 = getSelectedTab(execute=@n25.then) -> (index=0)
   > Update Tab Settings GUID(TabName=@n55.name, TabIndex=@n55, ControlRig=@n25.Current ControlRig Binding)
   n56 = getSelectedTab(execute=@n35.then) -> (index=0)
   > Update Tab Settings GUID(TabName=@n56.name, TabIndex=@n56, ControlRig=@n35.Current ControlRig Binding)
   n57 = Get NewTabTextBox!(self:EUW Animation Picker Tab Manager Object Reference=@n1) -> (NewTabTextBox:Editor Utility Editable Text Box Object Reference)
   addTab!(
    execute=@n3.then
    self
    tabName="To String (Text)"(InText=Get Text(self:Text Box Object Reference=@n57))
    tabWidgetToLoad=&Animation_Picker_C
    insertIndex=-1
    ValidateVisuals=true
    OptionalTabStyle
   ) -> (then, newTab:TCW Tab Data Struct Structure)
   float + float(A=@n22, B=35)
   n58 = getSelectedTab(execute=@n61.then) -> (index=0)
   > n59 = getTabWidgetByIndex(index=@n58)
   > n60 = Cast To Animation_Picker(Object=@n59)
   n61 = Cast To EUW_ContextMenu_ItemB(execute=@n9.then, Object=@n9.Slot Widget)
   n62 = "To String (Text)"(InText=Get Text(self:Text Object Reference=Get ButtonText(self=@n61)))
   n63 = Get Tabs ControlRig Binding(self=@n46)
   n64 = Find(TargetMap=@n63, Key=@n58.name)
   n65 = Set Bookmarks(execute=@n46.then, Bookmarks:Map of Strings to Widget Transform Structures=@n64.Value_Bookmarks_17_#9) -> (Output_Get:Map of Strings to Widget Transform Structures)
   > Add!(
    execute
    TargetMap:Map of Strings to Widget Transform Structures=@n65
    Key:"String (by ref)"=Replace(SourceString=@n62, From="Set ")
    Value:"Widget Transform Structure (by ref)"=Get RenderTransform!(self:Widget Object Reference=Get MoverCanvasPanel(self=@n60)) -> (RenderTransform:Widget Transform Structure)
   ) -> (then) {is=title K2Node_CallFunction}
   > n66 = Add!(
    execute
    TargetMap:Map of Strings to STRUCT Picker Tab Data Structures=@n63
    Key=@n58.name
    Value_ControlRigBinding_13_#7:Control Rig Sequencer Binding Proxy Structure=@n64
    Value_Index_16_#8:int=@n64.Value_Index_16_#8
    Value_Bookmarks_17_#9:Map of Strings to Widget Transform Structures=Get Bookmarks() -> (Bookmarks:Map of Strings to Widget Transform Structures)
   ) -> (then)
   n67 = Branch(execute=@n51.then, Condition=@n52.ReturnValue)
  ]}
 Get Control Rig Proxy Binding from GUID {nodes [
   n1 = Get Control Rig Proxy Binding from GUID!() -> (then, GUID:String) {is=title K2Node_FunctionEntry}
   > n2 = GetFocusedLevelSequence!(execute) -> (then, ReturnValue:Level Sequence Object Reference)
   > n3 = GetControlRigs!(execute, LevelSequence:Level Sequence Object Reference=@n2) -> (then, ReturnValue:Array of Control Rig Sequencer Binding Proxy Structures)
   > n4 = For Each Loop(Array:Array of Control Rig Sequencer Binding Proxy Structures=@n3) -> (Array Element:Control Rig Sequencer Binding Proxy Structure)
   Return Node!(execute=@n4.Completed, Control Rig Binding Proxy:Control Rig Sequencer Binding Proxy Structure)
   n5 = Break Control Rig Sequencer Binding Proxy!(ControlRigSequencerBindingProxy=@n4) -> (Proxy_BindingID:Guid Structure, Proxy_Sequence:Movie Scene Sequence Object Reference, ControlRig, Track)
   > Branch(Condition="Equal Exactly (String)"(A="To String (Guid)"!(InGuid:"Guid Structure (by ref)"=@n5) -> (ReturnValue:String), B=@n1))
   > Return Node(Control Rig Binding Proxy=@n4)
  ]}
 OnKeyDown {nodes [
   n1 = OnKeyDown!() -> (then, MyGeometry:Geometry Structure, InKeyEvent:Key Event Structure) {is=title K2Node_FunctionEntry}
   > n2 = Sequence!(execute) -> (then_0, then_1)
   > n3 = getSelectedTab() -> (index=0)
   > n4 = getTabWidgetByIndex(index=@n3)
   > n5 = Cast To Animation_Picker(Object=@n4)
   > n6 = LoadGameFromSlot(SlotName=Animation Picker, UserIndex=0)
   > n7 = Cast To SG_AnimationPicker(Object=@n6)
   n8 = Get Key Display Name!(
    Key:"Key Structure (by ref)"=GetKey!(Input:"Key Event Structure (by ref)"=@n1.InKeyEvent) -> (ReturnValue:Key Structure)
    bLongDisplayName:bool=true
   ) -> (ReturnValue:Text) {is=title K2Node_CallFunction}
   n9 = Find!(
    TargetMap:Map of Strings to Widget Transform Structures=Find!(TargetMap:Map of Strings to STRUCT Picker Tab Data Structures=Get Tabs ControlRig Binding(self=@n7), Key=@n3.name) -> (
     Value_ControlRigBinding_13_#7:Control Rig Sequencer Binding Proxy Structure
     Value_Index_16_#8:int
     Value_Bookmarks_17_#9:Map of Strings to Widget Transform Structures
     ReturnValue
    ).Value_Bookmarks_17_#9
    Key=Append(A=Bookmark, B="To String (Text)"(InText=@n8))
   ) -> (Value:Widget Transform Structure, ReturnValue)
   > Branch(Condition=@n9.ReturnValue)
   > SetRenderTransform(self=Get MoverCanvasPanel(self=@n5), InTransform=@n9)
   Return Node!(execute=@n2.then_1, ReturnValue:Event Reply Structure=Handled())
  ]}
 OnMouseButtonDown {nodes [
   OnMouseButtonDown!() -> (then, MyGeometry:Geometry Structure, MouseEvent:"Pointer Event Structure (by ref)") {is=title K2Node_FunctionEntry}
   > Show(self=Get SetBookmark_ContextMenu(), Show=false)
   > Return Node(ReturnValue=Handled())
  ]}
 Update Tab Settings GUID {nodes [
   n1 = Update Tab Settings GUID!() -> (then, TabName:String, TabIndex:int, ControlRig:Control Rig Sequencer Binding Proxy Structure) {is=title K2Node_FunctionEntry}
   > n2 = LoadGameFromSlot(SlotName=Animation Picker, UserIndex=0)
   > n3 = Cast To SG_AnimationPicker(Object=@n2)
   n4 = Get Tabs ControlRig Binding(self=@n3)
   n5 = Find!(TargetMap=@n4, Key=@n1) -> (Value:STRUCT Picker Tab Data Structure, ReturnValue)
   > Add!(
    execute
    TargetMap=@n4
    Key=@n1
    Value:"STRUCT Picker Tab Data Structure (by ref)"=Make STRUCT Picker Tab Data!(
     ControlRigBinding_13_#7:Control Rig Sequencer Binding Proxy Structure=@n1.ControlRig
     Index_16_#8:int=@n1.TabIndex
     Bookmarks_17_#9:Map of Strings to Widget Transform Structures=Break STRUCT Picker Tab Data(STRUCT_PickerTabData=@n5).Bookmarks_17_#9
    ) -> (STRUCT_PickerTabData:STRUCT Picker Tab Data Structure) {is=title K2Node_MakeStruct}
   ) -> (then)
   > SaveGameToSlot(SaveGameObject=@n3, SlotName=Animation Picker, UserIndex=0)
  ]}
]
variables [Bookmarks {type="map<WidgetTransform>", default {}}]
widgets {
 info {
  parentClass=&/EUW_BaseTabContainer.EUW_BaseTabContainer_C
  rootWidgetClass=&Border
  widgetCount=22
 }
 widgets [
  Border_1 {widgetClassPath=&Border, widget=Border_1, properties {brushColor=<rgba 0.0075 0.0075 0.0075 1>, padding=<ltrb 5 5 5 5>}}
  VerticalBox_24 {
   widgetClassPath=&VerticalBox
   parent=Border_1
   widget=VerticalBox_24
   slot {class=BorderSlot, properties {padding=<ltrb 5 5 5 5>}}
  }
  Overlay {
   widgetClassPath=&Overlay
   parent=VerticalBox_24
   widget=Overlay
   slot {class=VerticalBoxSlot, properties {size {sizeRule=Fill, value=1}}}
  }
  VerticalBox_0 {
   widgetClassPath=&VerticalBox
   parent=Overlay
   widget=VerticalBox_0
   slot {class=OverlaySlot, properties {horizontalAlignment=HAlign_Fill, verticalAlignment=VAlign_Fill}}
  }
  HorizontalBox_55 {
   widgetClassPath=&HorizontalBox
   parent=VerticalBox_0
   widget=HorizontalBox_55
   slot {class=VerticalBoxSlot, properties {horizontalAlignment=HAlign_Right}}
  }
  Tabs_TextBlock {
   widgetClassPath=&TextBlock
   parent=HorizontalBox_55
   widget=Tabs_TextBlock
   properties {
    text=Tabs:
    font {
     bForceMonospaced=false
     bMaterialIsStencil=false
     fontMaterial=null
     fontObject=&Roboto
     letterSpacing=0
     monospacedWidth=1
     size=9
     skewAmount=0
     typefaceFontName=Regular
     outlineSettings {
      bApplyOutlineToDropShadows=false
      bMiteredCorners=false
      bSeparateFillAlpha=false
      outlineMaterial=null
      outlineSize=0
      outlineColor=<rgba 0 0 0 1>
     }
    }
   }
   slot {class=HorizontalBoxSlot, properties {padding=<ltrb 0 0 18 0>}}
  }
  Tabs_TextBlock_1 {
   widgetClassPath=&TextBlock
   parent=HorizontalBox_55
   widget=Tabs_TextBlock_1
   properties {
    text=Bookmarks:
    font {
     bForceMonospaced=false
     bMaterialIsStencil=false
     fontMaterial=null
     fontObject=&Roboto
     letterSpacing=0
     monospacedWidth=1
     size=9
     skewAmount=0
     typefaceFontName=Regular
     outlineSettings {
      bApplyOutlineToDropShadows=false
      bMiteredCorners=false
      bSeparateFillAlpha=false
      outlineMaterial=null
      outlineSize=0
      outlineColor=<rgba 0 0 0 1>
     }
    }
   }
   slot {class=HorizontalBoxSlot}
  }
  HorizontalBox_58 {widgetClassPath=&HorizontalBox, parent=VerticalBox_0, widget=HorizontalBox_58, slot {class=VerticalBoxSlot}}
  listTabs {
   widgetClassPath=/Game/Blueprints/Blutilities/ForTechArt/Widgets/List/TA_Generic01_List
   bIsVariable=true
   parent=HorizontalBox_58
   widget=listTabs
   properties {
    allowMultiSelect=false
    listItemClass=/Game/Blueprints/Blutilities/ForTechArt/Widgets/Buttons/TA_Generic01_TabItem.TA_Generic01_TabItem_C
    scrollOrientation=Orient_Horizontal
   }
   slot {class=HorizontalBoxSlot, properties {size {sizeRule=Fill, value=1}}}
  }
  VerticalBox_1 {widgetClassPath=&VerticalBox, parent=HorizontalBox_58, widget=VerticalBox_1, slot {class=HorizontalBoxSlot}}
  HorizontalBox_53 {widgetClassPath=&HorizontalBox, parent=VerticalBox_1, widget=HorizontalBox_53, slot {class=VerticalBoxSlot}}
  AddTabButton {
   widgetClassPath=&EditorUtilityButton
   parent=HorizontalBox_53
   widget=AddTabButton
   properties {
    isFocusable=false
    toolTipText=Add New Tab
    widgetStyle {
     clickedSlateSound {resourceObject=null}
     disabled {
      bIsDynamicallyLoaded=false
      drawAs=RoundedBox
      imageType=NoImage
      mirroring=NoMirror
      resourceName=None
      resourceObject=null
      tiling=NoTile
      imageSize=<20 20>
      margin=<ltrb 0 0 0 0>
      outlineSettings {
       bUseBrushTransparency=false
       roundingType=FixedRadius
       width=1
       color {colorUseRule=UseColor_ColorTable, specifiedColor=<rgba 0 0 0 0>}
       cornerRadii=<4 4 4 4>
      }
      tintColor {colorUseRule=UseColor_Specified, specifiedColor=<rgba 0.0395 0.0395 0.0395 1>}
      uVRegion {bIsValid=false, max=<0 0>, min=<0 0>}
     }
     disabledForeground {colorUseRule=UseColor_ColorTable, specifiedColor=<rgba 0 0 0 0>}
     hovered {
      bIsDynamicallyLoaded=false
      drawAs=RoundedBox
      imageType=FullColor
      mirroring=NoMirror
      resourceName=None
      resourceObject=null
      tiling=NoTile
      imageSize=<20 20>
      margin=<ltrb 0 0 0 0>
      outlineSettings {
       bUseBrushTransparency=false
       roundingType=FixedRadius
       width=1
       color {colorUseRule=UseColor_ColorTable, specifiedColor=<rgba 0 0 0 0>}
       cornerRadii=<4 4 4 4>
      }
      tintColor {colorUseRule=UseColor_Specified, specifiedColor=<rgba 0.0501 0.5419 0 1>}
      uVRegion {bIsValid=false, max=<0 0>, min=<0 0>}
     }
     hoveredForeground {colorUseRule=UseColor_ColorTable, specifiedColor=<rgba 0 0 0 0>}
     hoveredSlateSound {resourceObject=null}
     normal {
      bIsDynamicallyLoaded=false
      drawAs=RoundedBox
      imageType=FullColor
      mirroring=NoMirror
      resourceName=None
      resourceObject=null
      tiling=NoTile
      imageSize=<20 20>
      margin=<ltrb 0 0 0 0>
      outlineSettings {
       bUseBrushTransparency=false
       roundingType=FixedRadius
       width=0
       color {colorUseRule=UseColor_ColorTable, specifiedColor=<rgba 0 0 0 0>}
       cornerRadii=<0 0 0 0>
      }
      tintColor {colorUseRule=UseColor_Specified, specifiedColor=<rgba 0.0871 0.9433 0 1>}
      uVRegion {bIsValid=false, max=<0 0>, min=<0 0>}
     }
     normalForeground {colorUseRule=UseColor_ColorTable, specifiedColor=<rgba 0 0 0 0>}
     normalPadding=<ltrb 12 1.5 12 1.5>
     pressed {
      bIsDynamicallyLoaded=false
      drawAs=RoundedBox
      imageType=FullColor
      mirroring=NoMirror
      resourceName=None
      resourceObject=null
      tiling=NoTile
      imageSize=<20 20>
      margin=<ltrb 0 0 0 0>
      outlineSettings {
       bUseBrushTransparency=false
       roundingType=FixedRadius
       width=1
       color {colorUseRule=UseColor_ColorTable, specifiedColor=<rgba 0 0 0 0>}
       cornerRadii=<4 4 4 4>
      }
      tintColor {colorUseRule=UseColor_Specified, specifiedColor=<rgba 0.2486 0.8464 0.1877 1>}
      uVRegion {bIsValid=false, max=<0 0>, min=<0 0>}
     }
     pressedForeground {colorUseRule=UseColor_ColorTable, specifiedColor=<rgba 0 0 0 0>}
     pressedPadding=<ltrb 12 2.5 12 0.5>
     pressedSlateSound {resourceObject=null}
    }
   }
   slot {class=HorizontalBoxSlot, properties {padding=<ltrb 2 2 2 2>}}
  }
  RemoveTabButton {
   widgetClassPath=&EditorUtilityButton
   parent=HorizontalBox_53
   widget=RemoveTabButton
   properties {
    isFocusable=false
    toolTipText=Remove Selected Tab
    widgetStyle {
     clickedSlateSound {resourceObject=null}
     disabled {
      bIsDynamicallyLoaded=false
      drawAs=RoundedBox
      imageType=NoImage
      mirroring=NoMirror
      resourceName=None
      resourceObject=null
      tiling=NoTile
      imageSize=<20 20>
      margin=<ltrb 0 0 0 0>
      outlineSettings {
       bUseBrushTransparency=false
       roundingType=FixedRadius
       width=1
       color {colorUseRule=UseColor_ColorTable, specifiedColor=<rgba 0 0 0 0>}
       cornerRadii=<4 4 4 4>
      }
      tintColor {colorUseRule=UseColor_Specified, specifiedColor=<rgba 0.0395 0.0395 0.0395 1>}
      uVRegion {bIsValid=false, max=<0 0>, min=<0 0>}
     }
     disabledForeground {colorUseRule=UseColor_ColorTable, specifiedColor=<rgba 0 0 0 0>}
     hovered {
      bIsDynamicallyLoaded=false
      drawAs=RoundedBox
      imageType=FullColor
      mirroring=NoMirror
      resourceName=None
      resourceObject=null
      tiling=NoTile
      imageSize=<20 20>
      margin=<ltrb 0 0 0 0>
      outlineSettings {
       bUseBrushTransparency=false
       roundingType=FixedRadius
       width=1
       color {colorUseRule=UseColor_ColorTable, specifiedColor=<rgba 0 0 0 0>}
       cornerRadii=<4 4 4 4>
      }
      tintColor {colorUseRule=UseColor_Specified, specifiedColor=<rgba 0.5419 0 0.0221 1>}
      uVRegion {bIsValid=false, max=<0 0>, min=<0 0>}
     }
     hoveredForeground {colorUseRule=UseColor_ColorTable, specifiedColor=<rgba 0 0 0 0>}
     hoveredSlateSound {resourceObject=null}
     normal {
      bIsDynamicallyLoaded=false
      drawAs=RoundedBox
      imageType=FullColor
      mirroring=NoMirror
      resourceName=None
      resourceObject=null
      tiling=NoTile
      imageSize=<20 20>
      margin=<ltrb 0 0 0 0>
      outlineSettings {
       bUseBrushTransparency=false
       roundingType=FixedRadius
       width=0
       color {colorUseRule=UseColor_ColorTable, specifiedColor=<rgba 0 0 0 0>}
       cornerRadii=<0 0 0 0>
      }
      tintColor {colorUseRule=UseColor_Specified, specifiedColor=<rgba 0.9433 0 0.0238 1>}
      uVRegion {bIsValid=false, max=<0 0>, min=<0 0>}
     }
     normalForeground {colorUseRule=UseColor_ColorTable, specifiedColor=<rgba 0 0 0 0>}
     normalPadding=<ltrb 12 1.5 12 1.5>
     pressed {
      bIsDynamicallyLoaded=false
      drawAs=RoundedBox
      imageType=FullColor
      mirroring=NoMirror
      resourceName=None
      resourceObject=null
      tiling=NoTile
      imageSize=<20 20>
      margin=<ltrb 0 0 0 0>
      outlineSettings {
       bUseBrushTransparency=false
       roundingType=FixedRadius
       width=1
       color {colorUseRule=UseColor_ColorTable, specifiedColor=<rgba 0 0 0 0>}
       cornerRadii=<4 4 4 4>
      }
      tintColor {colorUseRule=UseColor_Specified, specifiedColor=<rgba 0.8464 0.2108 0.2529 1>}
      uVRegion {bIsValid=false, max=<0 0>, min=<0 0>}
     }
     pressedForeground {colorUseRule=UseColor_ColorTable, specifiedColor=<rgba 0 0 0 0>}
     pressedPadding=<ltrb 12 2.5 12 0.5>
     pressedSlateSound {resourceObject=null}
    }
   }
   slot {class=HorizontalBoxSlot, properties {padding=<ltrb 5 0 18 0>}}
  }
  VerticalBox {widgetClassPath=&VerticalBox, parent=HorizontalBox_58, widget=VerticalBox, slot {class=HorizontalBoxSlot}}
  HorizontalBox_1 {widgetClassPath=&HorizontalBox, parent=VerticalBox, widget=HorizontalBox_1, slot {class=VerticalBoxSlot}}
  SetBookmarkButton {
   widgetClassPath=&EditorUtilityButton
   parent=HorizontalBox_1
   widget=SetBookmarkButton
   properties {
    isFocusable=false
    toolTipText=Set Bookmark
    widgetStyle {
     clickedSlateSound {resourceObject=null}
     disabled {
      bIsDynamicallyLoaded=false
      drawAs=RoundedBox
      imageType=NoImage
      mirroring=NoMirror
      resourceName=None
      resourceObject=null
      tiling=NoTile
      imageSize=<20 20>
      margin=<ltrb 0 0 0 0>
      outlineSettings {
       bUseBrushTransparency=false
       roundingType=FixedRadius
       width=1
       color {colorUseRule=UseColor_ColorTable, specifiedColor=<rgba 0 0 0 0>}
       cornerRadii=<4 4 4 4>
      }
      tintColor {colorUseRule=UseColor_Specified, specifiedColor=<rgba 0.0395 0.0395 0.0395 1>}
      uVRegion {bIsValid=false, max=<0 0>, min=<0 0>}
     }
     disabledForeground {colorUseRule=UseColor_ColorTable, specifiedColor=<rgba 0 0 0 0>}
     hovered {
      bIsDynamicallyLoaded=false
      drawAs=RoundedBox
      imageType=FullColor
      mirroring=NoMirror
      resourceName=None
      resourceObject=null
      tiling=NoTile
      imageSize=<20 20>
      margin=<ltrb 0 0 0 0>
      outlineSettings {
       bUseBrushTransparency=false
       roundingType=FixedRadius
       width=1
       color {colorUseRule=UseColor_ColorTable, specifiedColor=<rgba 0 0 0 0>}
       cornerRadii=<4 4 4 4>
      }
      tintColor {colorUseRule=UseColor_Specified, specifiedColor=<rgba 0.0363 0.1866 0.3637 1>}
      uVRegion {bIsValid=false, max=<0 0>, min=<0 0>}
     }
     hoveredForeground {colorUseRule=UseColor_ColorTable, specifiedColor=<rgba 0 0 0 0>}
     hoveredSlateSound {resourceObject=null}
     normal {
      bIsDynamicallyLoaded=false
      drawAs=RoundedBox
      imageType=FullColor
      mirroring=NoMirror
      resourceName=None
      resourceObject=null
      tiling=NoTile
      imageSize=<20 20>
      margin=<ltrb 0 0 0 0>
      outlineSettings {
       bUseBrushTransparency=false
       roundingType=FixedRadius
       width=0
       color {colorUseRule=UseColor_ColorTable, specifiedColor=<rgba 0 0 0 0>}
       cornerRadii=<0 0 0 0>
      }
      tintColor {colorUseRule=UseColor_Specified, specifiedColor=<rgba 0.2016 0.5841 0.9473 1>}
      uVRegion {bIsValid=false, max=<0 0>, min=<0 0>}
     }
     normalForeground {colorUseRule=UseColor_ColorTable, specifiedColor=<rgba 0 0 0 0>}
     normalPadding=<ltrb 12 1.5 12 1.5>
     pressed {
      bIsDynamicallyLoaded=false
      drawAs=RoundedBox
      imageType=FullColor
      mirroring=NoMirror
      resourceName=None
      resourceObject=null
      tiling=NoTile
      imageSize=<20 20>
      margin=<ltrb 0 0 0 0>
      outlineSettings {
       bUseBrushTransparency=false
       roundingType=FixedRadius
       width=1
       color {colorUseRule=UseColor_ColorTable, specifiedColor=<rgba 0 0 0 0>}
       cornerRadii=<4 4 4 4>
      }
      tintColor {colorUseRule=UseColor_Specified, specifiedColor=<rgba 0.5468 0.7522 0.9473 1>}
      uVRegion {bIsValid=false, max=<0 0>, min=<0 0>}
     }
     pressedForeground {colorUseRule=UseColor_ColorTable, specifiedColor=<rgba 0 0 0 0>}
     pressedPadding=<ltrb 12 2.5 12 0.5>
     pressedSlateSound {resourceObject=null}
    }
   }
   slot {class=HorizontalBoxSlot, properties {padding=<ltrb 2 2 2 2>}}
  }
  GoToBookmarkButton {
   widgetClassPath=&EditorUtilityButton
   parent=HorizontalBox_1
   widget=GoToBookmarkButton
   properties {
    isFocusable=false
    toolTipText=Go To Bookmark
    widgetStyle {
     clickedSlateSound {resourceObject=null}
     disabled {
      bIsDynamicallyLoaded=false
      drawAs=RoundedBox
      imageType=NoImage
      mirroring=NoMirror
      resourceName=None
      resourceObject=null
      tiling=NoTile
      imageSize=<0 0>
      margin=<ltrb 0 0 0 0>
      outlineSettings {
       bUseBrushTransparency=false
       roundingType=FixedRadius
       width=1
       color {colorUseRule=UseColor_ColorTable, specifiedColor=<rgba 0 0 0 0>}
       cornerRadii=<4 4 4 4>
      }
      tintColor {colorUseRule=UseColor_Specified, specifiedColor=<rgba 0.0395 0.0395 0.0395 1>}
      uVRegion {bIsValid=false, max=<0 0>, min=<0 0>}
     }
     disabledForeground {colorUseRule=UseColor_ColorTable, specifiedColor=<rgba 0 0 0 0>}
     hovered {
      bIsDynamicallyLoaded=false
      drawAs=RoundedBox
      imageType=FullColor
      mirroring=NoMirror
      resourceName=None
      resourceObject=null
      tiling=NoTile
      imageSize=<16 16>
      margin=<ltrb 0 0 0 0>
      outlineSettings {
       bUseBrushTransparency=false
       roundingType=FixedRadius
       width=1
       color {colorUseRule=UseColor_ColorTable, specifiedColor=<rgba 0 0 0 0>}
       cornerRadii=<4 4 4 4>
      }
      tintColor {colorUseRule=UseColor_Specified, specifiedColor=<rgba 0.2637 0.2637 0.2637 1>}
      uVRegion {bIsValid=false, max=<0 0>, min=<0 0>}
     }
     hoveredForeground {colorUseRule=UseColor_ColorTable, specifiedColor=<rgba 0 0 0 0>}
     hoveredSlateSound {resourceObject=null}
     normal {
      bIsDynamicallyLoaded=false
      drawAs=RoundedBox
      imageType=FullColor
      mirroring=NoMirror
      resourceName=None
      resourceObject=null
      tiling=NoTile
      imageSize=<16 16>
      margin=<ltrb 0 0 0 0>
      outlineSettings {
       bUseBrushTransparency=false
       roundingType=FixedRadius
       width=0
       color {colorUseRule=UseColor_ColorTable, specifiedColor=<rgba 0 0 0 0>}
       cornerRadii=<0 0 0 0>
      }
      tintColor {colorUseRule=UseColor_Specified, specifiedColor=<rgba 0.7183 0.7183 0.7183 1>}
      uVRegion {bIsValid=false, max=<0 0>, min=<0 0>}
     }
     normalForeground {colorUseRule=UseColor_ColorTable, specifiedColor=<rgba 0 0 0 0>}
     normalPadding=<ltrb 12 1.5 12 1.5>
     pressed {
      bIsDynamicallyLoaded=false
      drawAs=RoundedBox
      imageType=FullColor
      mirroring=NoMirror
      resourceName=None
      resourceObject=null
      tiling=NoTile
      imageSize=<16 16>
      margin=<ltrb 0 0 0 0>
      outlineSettings {
       bUseBrushTransparency=false
       roundingType=FixedRadius
       width=1
       color {colorUseRule=UseColor_ColorTable, specifiedColor=<rgba 0 0 0 0>}
       cornerRadii=<4 4 4 4>
      }
      tintColor {colorUseRule=UseColor_Specified, specifiedColor=<rgba 1 1 1 1>}
      uVRegion {bIsValid=false, max=<0 0>, min=<0 0>}
     }
     pressedForeground {colorUseRule=UseColor_ColorTable, specifiedColor=<rgba 0 0 0 0>}
     pressedPadding=<ltrb 12 2.5 12 0.5>
     pressedSlateSound {resourceObject=null}
    }
   }
   slot {class=HorizontalBoxSlot, properties {padding=<ltrb 6 0 10 0>}}
  }
  Border_2 {
   widgetClassPath=&Border
   parent=VerticalBox_0
   widget=Border_2
   properties {
    background {
     bIsDynamicallyLoaded=false
     drawAs=Image
     imageType=NoImage
     mirroring=NoMirror
     resourceName=None
     resourceObject=null
     tiling=NoTile
     imageSize=<32 32>
     margin=<ltrb 0 0 0 0>
     outlineSettings {
      bUseBrushTransparency=false
      roundingType=HalfHeightRadius
      width=0
      color {colorUseRule=UseColor_Specified, specifiedColor=<rgba 0 0 0 0>}
      cornerRadii=<0 0 0 0>
     }
     tintColor {colorUseRule=UseColor_Specified, specifiedColor=<rgba 0.0176 0.0176 0.0176 1>}
     uVRegion {bIsValid=false, max=<0 0>, min=<0 0>}
    }
    padding=<ltrb 0 0 0 0>
   }
   slot {class=VerticalBoxSlot, properties {size {sizeRule=Fill, value=1}}}
  }
  widgetSwitcherTabWidgets {
   widgetClassPath=&CommonWidgetSwitcherLegacy
   bIsVariable=true
   parent=Border_2
   widget=widgetSwitcherTabWidgets
   properties {transitionDuration=0.15}
   slot {class=BorderSlot, properties {padding=<ltrb 0 0 0 0>}}
  }
  DialogSlot {
   widgetClassPath=&NamedSlot
   bIsVariable=true
   parent=Overlay
   widget=DialogSlot
   slot {class=OverlaySlot, properties {horizontalAlignment=HAlign_Right}}
  }
  DialogSlot_1 {
   widgetClassPath=&NamedSlot
   bIsVariable=true
   parent=Overlay
   widget=DialogSlot_1
   slot {class=OverlaySlot, properties {horizontalAlignment=HAlign_Right}}
  }
  SetBookmark_ContextMenu {
   widgetClassPath=/AnimInEdTools/ContextMenu/EUW_ContextMenu
   bIsVariable=true
   parent=DialogSlot_1
   widget=SetBookmark_ContextMenu
   slot {class=PanelSlot}
  }
 ]
}
